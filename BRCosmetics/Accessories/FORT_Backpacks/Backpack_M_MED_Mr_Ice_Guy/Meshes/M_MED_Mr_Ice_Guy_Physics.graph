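# Physics asset for the Mr Ice Guy backpack cosmetic, giving its dangling tube and zipper physical bodies against the pack.
# A box body stands in for the pack, a small box for the zipper bone and a capsule for the tube bone.
# Two constraints tie the tube and zipper to the pack with limited swing and twist, the zipper held tighter than the tube.
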PhysicsAsset /BRCosmetics/Accessories/FORT_Backpacks/Backpack_M_MED_Mr_Ice_Guy/Meshes/M_MED_Mr_Ice_Guy_Physics
bodies {
 dyn_Tube [{shapeType=Capsule, length=1, radius=3, center=<0 0.2 -1>}]
 dyn_Zipper [{shapeType=Box, extentX=4.5, extentY=2, extentZ=1, center=<2.1854 0 -0.0001>}]
 pack [{shapeType=Box, extentX=21, extentY=13.5, extentZ=29.5, center=<0 -6.4227 -8.0706>}]
}
constraints [
 {bone1Name=dyn_Zipper, bone2Name=pack, swing1LimitDegrees=20, swing2LimitDegrees=30, twistLimitDegrees=10}
 {bone1Name=dyn_Tube, bone2Name=pack, swing1LimitDegrees=45, swing2LimitDegrees=45, twistLimitDegrees=45}
]
